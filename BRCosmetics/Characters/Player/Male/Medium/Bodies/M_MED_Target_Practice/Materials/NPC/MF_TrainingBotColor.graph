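# Material function that recolors a training bot's base color, with the color variant chosen per primitive through custom primitive data.
# The data value and TotalColors become a curve time that samples primary, secondary and tertiary rows in a shared color atlas.
# Each color multiplies the incoming BaseColor, blended in sequence by the R, G and B channels of a mask texture.
# Other material attributes pass through unchanged.
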MaterialFunction /BRCosmetics/Characters/Player/Male/Medium/Bodies/M_MED_Target_Practice/Materials/NPC/MF_TrainingBotColor
graph {nodes [
  n1 = GetMaterialAttributes(None=FunctionInput() {inputName=Material, inputType=FunctionInput_MaterialAttributes}) {attributeGetTypes [#1]}
  n2 = ScalarParameter() {defaultValue=50, parameterName=TotalColors}
  n3 = Divide(A=Constant() {r=1}, B=@n2)
  n4 = Add(A=Divide(A=ScalarParameter() {bUseCustomPrimitiveData=true, parameterName=PrimitiveData}, B=@n2), B=@n3)
  n5 = CurveAtlasRowParameter(CurveTime=@n4) {
   atlas=/BRCosmetics/Characters/Player/Male/Medium/Bodies/M_MED_Target_Practice/Textures/NPC/TrainingBotColorAtlas
   curve=/BRCosmetics/Characters/Player/Male/Medium/Bodies/M_MED_Target_Practice/Textures/NPC/TrainingBotColor_Primary
   parameterName=PrimaryColor
  }
  n6 = TextureSampleParameter2D() {
   parameterName=T_Bot_Color_Map
   texture=/BRCosmetics/Characters/Player/Male/Medium/Bodies/M_MED_Target_Practice/Textures/NPC/M_MED_Training_Ground_Bot_Body_TA
  }
  n7 = CurveAtlasRowParameter(CurveTime=@n4) {
   atlas=&TrainingBotColorAtlas
   curve=/BRCosmetics/Characters/Player/Male/Medium/Bodies/M_MED_Target_Practice/Textures/NPC/TrainingBotColor_Secondary
   defaultValue=1
   parameterName=SecondaryColor
  }
  n8 = LinearInterpolate(A=LinearInterpolate(A=@n1.BaseColor, B=Multiply(A=@n1.BaseColor, B=@n5), Alpha=@n6.R), B=Multiply(A=@n1.BaseColor, B=@n7), Alpha=@n6.G)
  n9 = CurveAtlasRowParameter(CurveTime=@n4) {
   atlas=&TrainingBotColorAtlas
   curve=/BRCosmetics/Characters/Player/Male/Medium/Bodies/M_MED_Target_Practice/Textures/NPC/TrainingBotColor_Tertiary
   defaultValue=2
   parameterName=TertiaryColor
  }
  n10 = Multiply(A=@n1.BaseColor, B=@n9)
  FunctionOutput(None=SetMaterialAttributes!(MaterialAttributes=@n1.MaterialAttributes, BaseColor=LinearInterpolate(A=@n8, B=@n10, Alpha=@n6.B)) {attributeSetTypes [#1]})
  Multiply(A=@n3) {constB=0.5}
 ]}
